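# Physics asset for the Banana Agent medium male body, covering the dynamic coat bones plus pelvis and thigh collision shapes.
# Nine two-segment capsule chains around the back, sides and front hang from tiny box base bodies constrained to the pelvis.
# Chain joints allow wide swing on one axis, with tight second-swing and twist limits.
# An angularly free constraint links the right front and front-side tips, with no left-side equivalent.
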PhysicsAsset /BRCosmetics/Characters/Player/Male/Medium/Bodies/M_MED_Banana_Agent/Meshes/M_MED_BananaAgent_Physics
bodies {
 dyn_coatBack_01_Base_l [{shapeType=Box, extentX=0.5198, extentY=0.5, extentZ=0.5002, rotation=<rot 0 -90 0>}]
 dyn_coatBack_01_Base_r [{shapeType=Box, extentX=0.5198, extentY=0.5, extentZ=0.5002, rotation=<rot 0 -90 0>}]
 dyn_coatBack_01_l [{shapeType=Capsule, length=13.25, radius=4.25, shapeName=dyn_coatBack_01_r_capsule, center=<-4 2 0>, rotation=<rot 0 0 -169.9993>}]
 dyn_coatBack_01_r [{shapeType=Capsule, length=13.25, radius=4.25, shapeName=dyn_coatBack_01_r_capsule, center=<4 -2 0>, rotation=<rot 0 0 10.0007>}]
 dyn_coatBack_01a [{shapeType=Capsule, length=13.5, radius=4, shapeName=dyn_coatBack_01a_capsule, center=<4 -1 0>}]
 dyn_coatBack_01a_Base [{shapeType=Box, extentX=0.5198, extentY=0.5, extentZ=0.5002, rotation=<rot 0 -90 0>}]
 dyn_coatBack_02_l [{shapeType=Capsule, length=13.25, radius=4.25, shapeName=dyn_coatBack_02_r_capsule, center=<-4 2 0>, rotation=<rot 0 0 -169.9993>}]
 dyn_coatBack_02_r [{shapeType=Capsule, length=13.25, radius=4.25, shapeName=dyn_coatBack_02_r_capsule, center=<4 -2 0>, rotation=<rot 0 0 10.0007>}]
 dyn_coatBack_02a [{shapeType=Capsule, length=13.5, radius=4, shapeName=dyn_coatBack_02a_capsule, center=<4 -1 0>}]
 dyn_coatSide_01_Base_l [{shapeType=Box, extentX=0.5198, extentY=0.5, extentZ=0.5002, rotation=<rot 0 -90 0>}]
 dyn_coatSide_01_Base_r [{shapeType=Box, extentX=0.5198, extentY=0.5, extentZ=0.5002, rotation=<rot 0 -90 0>}]
 dyn_coatSide_01_l [{shapeType=Capsule, length=13.5, radius=4, shapeName=dyn_coatSide_01_r_capsule, center=<-4 3 0>, rotation=<rot 0 0 180>}]
 dyn_coatSide_01_r [{shapeType=Capsule, length=13.5, radius=4, shapeName=dyn_coatSide_01_r_capsule, center=<4 -3 0>}]
 dyn_coatSide_02_l [{shapeType=Capsule, length=13.5, radius=4, shapeName=dyn_coatSide_02_r_capsule, center=<-4 3 0>, rotation=<rot 0 0 180>}]
 dyn_coatSide_02_r [{shapeType=Capsule, length=13.5, radius=4, shapeName=dyn_coatSide_02_r_capsule, center=<4 -3 0>}]
 dyn_coatSide_fr_01_Base_l [{shapeType=Box, extentX=0.5, extentY=0.5, extentZ=0.5, rotation=<rot 0 90 0>}]
 dyn_coatSide_fr_01_Base_r [{shapeType=Box, extentX=0.5, extentY=0.5, extentZ=0.5, rotation=<rot 0 90 0>}]
 dyn_coatSide_fr_01_l [{shapeType=Capsule, length=14.25, radius=4.75, shapeName=dyn_coatSide_fr_01_r_capsule, center=<-6 3.5 0>, rotation=<rot 0 0 169.9993>}]
 dyn_coatSide_fr_01_r [{shapeType=Capsule, length=14.25, radius=4.75, shapeName=dyn_coatSide_fr_01_r_capsule, center=<6 -3.5 0>, rotation=<rot 0 0 -10.0007>}]
 dyn_coatSide_fr_02_l [{shapeType=Capsule, length=14.25, radius=4.75, shapeName=dyn_coatSide_fr_02_r_capsule, center=<-4 3.5 0>, rotation=<rot 0 0 169.9993>}]
 dyn_coatSide_fr_02_r [{shapeType=Capsule, length=14.25, radius=4.75, shapeName=dyn_coatSide_fr_02_r_capsule, center=<4 -3.5 0>, rotation=<rot 0 0 -10.0007>}]
 dyn_coat_fr_01_Base_l [{shapeType=Box, extentX=0.5, extentY=0.5, extentZ=0.5, rotation=<rot 0 90 0>}]
 dyn_coat_fr_01_Base_r [{shapeType=Box, extentX=0.5, extentY=0.5, extentZ=0.5, rotation=<rot 0 90 0>}]
 dyn_coat_fr_01_l [{shapeType=Capsule, length=8, radius=5.5, shapeName=dyn_coat_fr_01_r_capsule, center=<-6 3 0>, rotation=<rot 0 0 180>}]
 dyn_coat_fr_01_r [{shapeType=Capsule, length=8, radius=5.5, shapeName=dyn_coat_fr_01_r_capsule, center=<6 -3 0>}]
 dyn_coat_fr_02_l [{shapeType=Capsule, length=8, radius=5.5, shapeName=dyn_coat_fr_02_r_capsule, center=<-6 3 0>, rotation=<rot 0 0 180>}]
 dyn_coat_fr_02_r [{shapeType=Capsule, length=8, radius=5.5, shapeName=dyn_coat_fr_02_r_capsule, center=<6 -3 0>}]
 pelvis [{shapeType=Capsule, length=1.5, radius=14.5, shapeName=pelvis_capsule, center=<0.0001 1.5257 0>, rotation=<rot 90 0 0>}]
 thigh_l [
  {shapeType=Capsule, length=27.9836, radius=8.9381, center=<-14.5356 2.1922 1.5448>, rotation=<rot 88.901 -167.6258 -0.0007>}
  {shapeType=Capsule, length=10.5409, radius=8.5387, center=<-7.6981 -1.8469 0.9076>, rotation=<rot 89.5817 95.1321 -77.685>}
 ]
 thigh_r [
  {shapeType=Capsule, length=27.9836, radius=8.9381, center=<14.536 -2.1921 -1.5683>, rotation=<rot 88.9008 -167.6258 179.9992>}
  {shapeType=Capsule, length=10.5409, radius=8.5387, center=<7.8018 2.5217 -1.0552>, rotation=<rot 84.8958 168.3018 175.4577>}
  {shapeType=Capsule, length=5.4781, radius=10.6387, center=<34.0851 -0.8647 -6.0892>, rotation=<rot 90 0 -8.4961>}
 ]
}
constraints [
 {bone1Name=dyn_coat_fr_01_r, bone2Name=dyn_coat_fr_01_Base_r, swing1LimitDegrees=45, swing2LimitDegrees=5, twistLimitDegrees=2}
 {bone1Name=dyn_coat_fr_02_r, bone2Name=dyn_coat_fr_01_r, swing1LimitDegrees=45, swing2LimitDegrees=5, twistLimitDegrees=2}
 {bone1Name=dyn_coatSide_fr_02_r, bone2Name=dyn_coatSide_fr_01_r, swing1LimitDegrees=45, swing2LimitDegrees=5, twistLimitDegrees=2}
 {bone1Name=dyn_coatSide_fr_01_r, bone2Name=dyn_coatSide_fr_01_Base_r, swing1LimitDegrees=45, swing2LimitDegrees=5, twistLimitDegrees=2}
 {
  bone1Name=dyn_coatSide_fr_02_r
  bone2Name=dyn_coat_fr_02_r
  swing1LimitDegrees=45
  swing1Motion=Free
  swing2LimitDegrees=45
  swing2Motion=Free
  twistLimitDegrees=45
  twistMotion=Free
 }
 {bone1Name=dyn_coatBack_01a, bone2Name=dyn_coatBack_01a_Base, swing1LimitDegrees=40, swing2LimitDegrees=5, twistLimitDegrees=2}
 {bone1Name=dyn_coatBack_02a, bone2Name=dyn_coatBack_01a, swing1LimitDegrees=40, swing2LimitDegrees=5, twistLimitDegrees=2}
 {bone1Name=dyn_coatBack_01_r, bone2Name=dyn_coatBack_01_Base_r, swing1LimitDegrees=45, swing2LimitDegrees=5, twistLimitDegrees=2}
 {bone1Name=dyn_coatBack_02_r, bone2Name=dyn_coatBack_01_r, swing1LimitDegrees=45, swing2LimitDegrees=5, twistLimitDegrees=2}
 {bone1Name=dyn_coatSide_01_r, bone2Name=dyn_coatSide_01_Base_r, swing1LimitDegrees=45, swing2LimitDegrees=5, twistLimitDegrees=2}
 {bone1Name=dyn_coatSide_02_r, bone2Name=dyn_coatSide_01_r, swing1LimitDegrees=45, swing2LimitDegrees=5, twistLimitDegrees=2}
 {
  bone1Name=thigh_l
  bone2Name=pelvis
  swing1LimitDegrees=45
  swing1Motion=Free
  swing2LimitDegrees=45
  swing2Motion=Free
  twistLimitDegrees=45
  twistMotion=Free
 }
 {
  bone1Name=thigh_r
  bone2Name=pelvis
  swing1LimitDegrees=45
  swing1Motion=Free
  swing2LimitDegrees=45
  swing2Motion=Free
  twistLimitDegrees=45
  twistMotion=Free
 }
 {
  bone1Name=dyn_coatBack_01a_Base
  bone2Name=pelvis
  swing1LimitDegrees=45
  swing1Motion=Free
  swing2LimitDegrees=45
  swing2Motion=Free
  twistLimitDegrees=45
  twistMotion=Free
 }
 {
  bone1Name=dyn_coatBack_01_Base_r
  bone2Name=pelvis
  swing1LimitDegrees=45
  swing1Motion=Free
  swing2LimitDegrees=45
  swing2Motion=Free
  twistLimitDegrees=45
  twistMotion=Free
 }
 {
  bone1Name=dyn_coatSide_01_Base_r
  bone2Name=pelvis
  swing1LimitDegrees=45
  swing1Motion=Free
  swing2LimitDegrees=45
  swing2Motion=Free
  twistLimitDegrees=45
  twistMotion=Free
 }
 {
  bone1Name=dyn_coatSide_fr_01_Base_r
  bone2Name=pelvis
  swing1LimitDegrees=45
  swing1Motion=Free
  swing2LimitDegrees=45
  swing2Motion=Free
  twistLimitDegrees=45
  twistMotion=Free
 }
 {
  bone1Name=dyn_coat_fr_01_Base_r
  bone2Name=pelvis
  swing1LimitDegrees=45
  swing1Motion=Free
  swing2LimitDegrees=45
  swing2Motion=Free
  twistLimitDegrees=45
  twistMotion=Free
 }
 {
  bone1Name=dyn_coatBack_01_Base_l
  bone2Name=pelvis
  swing1LimitDegrees=45
  swing1Motion=Free
  swing2LimitDegrees=45
  swing2Motion=Free
  twistLimitDegrees=45
  twistMotion=Free
 }
 {
  bone1Name=dyn_coatSide_01_Base_l
  bone2Name=pelvis
  swing1LimitDegrees=45
  swing1Motion=Free
  swing2LimitDegrees=45
  swing2Motion=Free
  twistLimitDegrees=45
  twistMotion=Free
 }
 {
  bone1Name=dyn_coatSide_fr_01_Base_l
  bone2Name=pelvis
  swing1LimitDegrees=45
  swing1Motion=Free
  swing2LimitDegrees=45
  swing2Motion=Free
  twistLimitDegrees=45
  twistMotion=Free
 }
 {
  bone1Name=dyn_coat_fr_01_Base_l
  bone2Name=pelvis
  swing1LimitDegrees=45
  swing1Motion=Free
  swing2LimitDegrees=45
  swing2Motion=Free
  twistLimitDegrees=45
  twistMotion=Free
 }
 {bone1Name=dyn_coatBack_01_l, bone2Name=dyn_coatBack_01_Base_l, swing1LimitDegrees=45, swing2LimitDegrees=5, twistLimitDegrees=2}
 {bone1Name=dyn_coatBack_02_l, bone2Name=dyn_coatBack_01_l, swing1LimitDegrees=45, swing2LimitDegrees=5, twistLimitDegrees=2}
 {bone1Name=dyn_coatSide_01_l, bone2Name=dyn_coatSide_01_Base_l, swing1LimitDegrees=45, swing2LimitDegrees=5, twistLimitDegrees=2}
 {bone1Name=dyn_coatSide_02_l, bone2Name=dyn_coatSide_01_l, swing1LimitDegrees=45, swing2LimitDegrees=5, twistLimitDegrees=2}
 {bone1Name=dyn_coatSide_fr_01_l, bone2Name=dyn_coatSide_fr_01_Base_l, swing1LimitDegrees=45, swing2LimitDegrees=5, twistLimitDegrees=2}
 {bone1Name=dyn_coatSide_fr_02_l, bone2Name=dyn_coatSide_fr_01_l, swing1LimitDegrees=45, swing2LimitDegrees=5, twistLimitDegrees=2}
 {bone1Name=dyn_coat_fr_01_l, bone2Name=dyn_coat_fr_01_Base_l, swing1LimitDegrees=45, swing2LimitDegrees=5, twistLimitDegrees=2}
 {bone1Name=dyn_coat_fr_02_l, bone2Name=dyn_coat_fr_01_l, swing1LimitDegrees=45, swing2LimitDegrees=5, twistLimitDegrees=2}
]
