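# Post-process style Anim Blueprint for the male Banana Agent body, adding procedural coat motion to the pose copied from the parent mesh.
# Control Rig receives a skydiving value, then a rotation constraint, pelvis-space RigidBody simulation and additive bone offset adjust the coat.
# Event updates feed RigidBody gravity from a deprecated state-based override, using standard gravity when DBNO.
# It also flags three specific emotes, though the AnimGraph ignores that flag.
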AnimBlueprint /BRCosmetics/Characters/Player/Male/Medium/Bodies/M_MED_Banana_Agent/Meshes/M_MED_BananaAgent_AnimBP
target_skeleton=&/M_MED_BananaAgent_Skeleton
graphs [
 AnimGraph {nodes [
   n1 = Control Rig!(Source=Copy Pose From Mesh() {bUseAttachedParent=true}, Alpha, skydiving:double=Get skydiving() -> (skydiving:double)) -> (Pose) {
    alphaBoolBlend {blendOption=Linear}
    controlRigAssetReference {blueprintRigClass=&/M_MED_BananaAgent_Body_CtrlRig.M_MED_BananaAgent_Body_CtrlRig_C}
   }
   n2 = "Constraint - dyn_coatMiddle_01_Base (RRR)"!(
    ComponentPose=Local To Component(LocalPose=@n1)
    Alpha=1
    ConstraintWeights_0=0.3
    ConstraintWeights_1:float=0.4
    ConstraintWeights_2:float=0.3
   ) -> (Pose) {
    alphaBoolBlend {blendOption=Linear}
    boneToModify {boneName=dyn_coatMiddle_01_Base}
    constraintSetup [
     {transformType=Rotation, targetBone {boneName=pelvis}}
     {transformType=Rotation, targetBone {boneName=dyn_coat_fr_01_Base_r}}
     {transformType=Rotation, targetBone {boneName=dyn_coat_fr_01_Base_l}}
    ]
    constraintWeights [1,1,1]
   }
   n3 = RigidBody!(ComponentPose=@n2, Alpha, OverrideWorldGravity:Vector=Get GravityOverride(), ExternalForce) -> (Pose) {
    bClampLinearTranslationLimitToRefPose=true
    bOverrideWorldGravity=true
    overridePhysicsAsset=&/M_MED_BananaAgent_Physics
    simulationSpace=BaseBoneSpace
    alphaBoolBlend {blendOption=Linear}
    baseBoneRef {boneName=pelvis}
    componentAppliedLinearAccClamp=<1000 1000 1000>
    componentLinearAccScale=<0.4 0.4 0.5>
    componentLinearVelScale=<0.2 0.2 0.25>
    overrideWorldGravity {z=-980}
   }
   n4 = "Transform (Modify) Bone - Bone: dyn_coatSide_01_Base_r"(ComponentPose=@n3, Alpha=1, Translation="0,-3,0", Rotation="0,0,0", Scale="1,1,1") {
    translationMode=BMM_Additive
    translationSpace=BCS_BoneSpace
    alphaBoolBlend {blendOption=Linear}
    boneToModify {boneName=dyn_coatSide_01_Base_r}
    translation {y=-3}
   }
   Output Pose(Result=Component To Local(ComponentPose=@n4)) {layerGroup=None}
   Break Fort Anim Input Skydiving External Force(FortAnimInput_SkydivingExternalForce=Get SkydivingExternalForceInput()) -> (
    bUseSkydivingVectorForce=~
    SkydivingVectorForce=~
    HeadToPelvisDirection=~
    FloatingMultiplier=~
    FloatingAdditive=~
    DivingMultiplier=~
    DivingAdditive=~
    ParachutingMultiplier=~
    ParachutingAdditive=~
   )
  ]}
 EventGraph {nodes [
   Event BlueprintUpdateAnimation()
   > n1 = Sequence!(execute) -> (then_0, then_1, then_2:Exec)
   > Set skydiving(skydiving:double="To Float (Boolean)"(InBool=Get bIsSkyDiving())) -> (Output_Get:double)
   n2 = GetAnimInstance(self=GetOwnerSkeletalMeshComponent())
   Is Valid(exec=@n1.then_1, InputObject=@n2)
   n3 = Montage_IsPlaying(self=@n2, Montage=/BRCosmetics/Animation/Game/MainPlayer/Emotes/Paws/Paws_CMM_M)
   n4 = Montage_IsPlaying(self=@n2, Montage=/BRCosmetics/Animation/Game/MainPlayer/Emotes/Hip_Hop/Emote_Hip_Hop_CMM_M)
   n5 = Montage_IsPlaying(self=@n2, Montage=/BRCosmetics/Animation/Game/MainPlayer/Emotes/Cowbell/Cowbell_CMM_Loop_M)
   > Set isCustomEmotePlaying(isCustomEmotePlaying:double=SelectFloat(A=1, B=0, bPickA=OR Boolean!(A=@n4, B=@n3, C:bool=@n5, D:bool, E:bool) -> (ReturnValue))) -> (Output_Get:double)
   n6 = Gravity Override State Function DEPRECATED(
    execute=@n1.then_2
    CharacterPartAnimInstance=Self-Reference()
    GameTimeInSeconds=GetGameTimeInSeconds()
    GravityOverrideParams=Get Gravity Override Params()
   )
   > Set GravityOverride(GravityOverride=@n6)
   > n7 = Branch(Condition=Get bIsDBNO())
   > Set GravityOverride(GravityOverride="0,0,-980")
   Set GravityOverride(execute=@n7.else, GravityOverride=@n6)
  ]}
]
variables [
 Gravity Override Params {
  type=GravityOverrideParamsStruct
  default {
   skydive_FloatingDown {gravityOverride=<0 -700 -300>, windAmplitude=<200 -500 100>, windFrequency=<1.5 10 15>, windOffset=<0 -600 100>}
   skydive_GoingUp {gravityOverride=<0 -300 -100>, windAmplitude=<200 -500 100>}
   skydive_Parachuting {gravityOverride=<0 -500 -980>, windAmplitude=<200 -300 400>}
  }
 }
 GravityOverride {type=Vector, default=<0 0 -980>}
 isCustomEmotePlaying {type=number}
 skydiving {type=number}
]
