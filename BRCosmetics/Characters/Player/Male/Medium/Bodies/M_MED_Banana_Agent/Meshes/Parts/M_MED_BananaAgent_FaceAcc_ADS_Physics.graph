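# Physics asset for the ADS face accessory of the male medium Banana Agent player cosmetic, covering its left and right dynamic flaps.
# Every body is a single box, with small cubes on the two flap base bones and larger fitted boxes on the flaps.
# The left side chains base to two flap segments, while the right joins one flap to its base.
# Each joint sets narrow swing and twist limit angles, with the widest swing on the outer left segment.
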PhysicsAsset /BRCosmetics/Characters/Player/Male/Medium/Bodies/M_MED_Banana_Agent/Meshes/Parts/M_MED_BananaAgent_FaceAcc_ADS_Physics
bodies {
 dyn_flapBase_r [{shapeType=Box, extentX=1.01, extentY=1.01, extentZ=1.01, rotation=<rot 0 -90 0>}]
 dyn_flap_01_l [{shapeType=Box, extentX=3.5716, extentY=7.2867, extentZ=15.0287, center=<4.0812 -0.4883 -1.3233>, rotation=<rot 19.2099 -90 4.4879>}]
 dyn_flap_02_l [{shapeType=Box, extentX=2.7009, extentY=10.1105, extentZ=11.7929, center=<4.2095 0.549 -0.5692>, rotation=<rot -2.5285 100.422 -100.8669>}]
 dyn_flap_Base_l [{shapeType=Box, extentX=1.01, extentY=1.01, extentZ=1.01, rotation=<rot 0 -90 0>}]
 dyn_flap_r [{shapeType=Box, extentX=19.323, extentY=3.3572, extentZ=15.5488, center=<7.3236 0 0>}]
}
constraints [
 {bone1Name=dyn_flap_r, bone2Name=dyn_flapBase_r, swing1LimitDegrees=30, swing2LimitDegrees=10, twistLimitDegrees=3}
 {bone1Name=dyn_flap_01_l, bone2Name=dyn_flap_Base_l, swing1LimitDegrees=30, swing2LimitDegrees=10, twistLimitDegrees=12}
 {bone1Name=dyn_flap_02_l, bone2Name=dyn_flap_01_l, swing1LimitDegrees=45, swing2LimitDegrees=10, twistLimitDegrees=12}
]
