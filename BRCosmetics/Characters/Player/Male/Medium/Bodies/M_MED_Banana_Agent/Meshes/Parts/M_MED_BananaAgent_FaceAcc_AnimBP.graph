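# Animation Blueprint for the Banana Agent face accessory, posing its dynamic flaps and mouth material.
# It layers body-mesh upper-arm bones over the attached parent's pose, then constrains both flap bases to spine and neck.
# Targeting state switches between normal and ADS RigidBody physics assets, with extra yaw on the right flap while targeting.
# Emote state sets the material's smile-mouth frame, on events and every update.
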AnimBlueprint /BRCosmetics/Characters/Player/Male/Medium/Bodies/M_MED_Banana_Agent/Meshes/Parts/M_MED_BananaAgent_FaceAcc_AnimBP
target_skeleton=&/M_MED_BananaAgent_FaceAcc_Skeleton
graphs [
 AnimGraph {nodes [
   n1 = "Constraint - dyn_flapBase_r (PP)"!(ComponentPose=Local To Component(LocalPose=Use cached pose 'Start_Cache'()), Alpha=1, ConstraintWeights_0=0.8, ConstraintWeights_1:float=0.2) -> (Pose) {
    alphaBoolBlend {blendOption=Linear}
    boneToModify {boneName=dyn_flapBase_r}
    constraintSetup [{targetBone {boneName=spine_05}}, {targetBone {boneName=neck_01}}]
    constraintWeights [1,1]
   }
   n2 = "Constraint - dyn_flap_Base_l (PP)"!(ComponentPose=@n1, Alpha=1, ConstraintWeights_0=0.8, ConstraintWeights_1:float=0.2) -> (Pose) {
    alphaBoolBlend {blendOption=Linear}
    boneToModify {boneName=dyn_flap_Base_l}
    constraintSetup [{targetBone {boneName=spine_05}}, {targetBone {boneName=neck_01}}]
    constraintWeights [1,1]
   }
   n3 = RigidBody(ComponentPose=@n2, Alpha=Get isNotTargeting() -> (isNotTargeting:double)) {
    bClampLinearTranslationLimitToRefPose=true
    bOverrideWorldGravity=true
    overridePhysicsAsset=&/M_MED_BananaAgent_FaceAcc_Physics
    simulationSpace=BaseBoneSpace
    alphaBoolBlend {blendOption=Linear}
    baseBoneRef {boneName=root}
    componentAppliedLinearAccClamp=<800 800 800>
    componentLinearAccScale=<0.65 0.65 0.65>
    componentLinearVelScale=<0.3 0.3 0.3>
    overrideWorldGravity {z=-980}
   }
   n4 = Get isTargeting() -> (isTargeting:double)
   n5 = RigidBody(ComponentPose=@n3, Alpha=@n4) {
    bClampLinearTranslationLimitToRefPose=true
    bOverrideWorldGravity=true
    overridePhysicsAsset=&/M_MED_BananaAgent_FaceAcc_ADS_Physics
    simulationSpace=BaseBoneSpace
    alphaBoolBlend {blendOption=Linear}
    baseBoneRef {boneName=root}
    componentAppliedLinearAccClamp=<800 800 800>
    componentLinearAccScale=<0.65 0.65 0.65>
    componentLinearVelScale=<0.3 0.3 0.3>
    overrideWorldGravity {z=-980}
   }
   n6 = "Transform (Modify) Bone - Bone: dyn_flapBase_r"(ComponentPose=@n5, Alpha=@n4, Translation="0,0,0", Rotation="0,-18,0", Scale="1,1,1") {
    rotationMode=BMM_Additive
    rotationSpace=BCS_BoneSpace
    translationSpace=BCS_BoneSpace
    alphaBoolBlend {blendOption=Linear}
    boneToModify {boneName=dyn_flapBase_r}
    rotation {yaw=-18}
   }
   Output Pose(Result=Component To Local(ComponentPose=@n6)) {layerGroup=None}
   n7 = Layered blend per bone(
    BasePose=Copy Pose From Mesh() {bCopyCurves=true, bUseAttachedParent=true}
    BlendPoses_0=Copy Pose From Mesh(SourceMeshComponent=Get Body_Mesh())
   ) {layerSetup [{branchFilters [{boneName=upperarm_r}, {boneName=upperarm_l}, {boneName=upperarm_twist_01_r}, {boneName=upperarm_twist_01_l}]}]}
   Save cached pose 'Start_Cache'(Pose=@n7) {cacheName=Start_Cache}
  ]}
 EventGraph {nodes [
   Event BlueprintUpdateAnimation()
   > n1 = Get Dynamic Material Instance(SkeletalMeshComponent=GetPartSkeletalMeshComponent())
   > n2 = Branch(Condition=Get bIsPlayingEmote())
   > SetScalarParameterValue(self=@n1, ParameterName=Smile Mouth animation Frame, Value=8)
   SetScalarParameterValue(execute=@n2.else, self=@n1, ParameterName=Smile Mouth animation Frame, Value=3)
   Event BlueprintInitializeAnimation()
   > Set Body_Mesh(Body_Mesh=GetOwnerSkeletalMeshComponent())
   > n3 = Cast To PlayerPawn_Athena(Object=TryGetPawnOwner())
   > n4 = Set Pawn Owner(Pawn Owner:Player Pawn Athena Object Reference=@n3) -> (Output_Get:Player Pawn Athena Object Reference)
   > Set Head_Mesh(Head_Mesh:Skeletal Mesh Component Object Reference=GetSkeletalMeshForPartType(self=@n4, PartType=Head)) -> (Output_Get:Skeletal Mesh Component Object Reference)
   n5 = Event OnBeginTargeting() {is=title K2Node_Event}
   n6 = "To Float (Boolean)"(InBool=Get bIsTargeting())
   Set isTargeting(execute=[@n5.then, @n7.then], isTargeting:double=@n6) -> (Output_Get:double)
   > Set isNotTargeting(isNotTargeting:double=float - float(B=@n6) {is=title K2Node_CallFunction}) -> (Output_Get:double)
   n7 = Event OnEndTargeting() {is=title K2Node_Event}
   n8 = Event OnBeginPlayingEmote() {is=title K2Node_Event}
   n9 = Get Dynamic Material Instance(execute=[@n8.then, @n10.then], SkeletalMeshComponent=GetPartSkeletalMeshComponent())
   > Is Valid(InputObject=@n9)
   > SetScalarParameterValue(self=@n9, ParameterName=Smile Mouth animation Frame, Value=SelectFloat(A=8, B=3, bPickA=Get bIsPlayingEmote()))
   n10 = Event OnEndPlayingEmote() {is=title K2Node_Event}
  ]}
]
variables [
 Body_Mesh {type=SkeletalMeshComponent}
 Head_Mesh {type=SkeletalMeshComponent}
 isNotTargeting {type=number}
 isTargeting {type=number}
 Pawn Owner {type=PlayerPawn_Athena_C}
]
